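# Physics asset for the M_MED_BlueJet backpack cosmetic, simulating its dangling carabiner, chain, stick and strap with capsule bodies.
# The root, pack and dyn_pack bodies are angularly locked in series, with dyn_pack carrying the large main capsule.
# From dyn_pack, a free-rotating master carabiner joint leads to a swing-restricted carabiner and four limited chain links.
# The stick and strap attach separately to dyn_pack with tighter angular limits.
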PhysicsAsset /BRCosmetics/Accessories/FORT_Backpacks/Backpack_M_MED_BlueJet/Meshes/M_MED_BlueJet_Pack_Physics1
bodies {
 dyn_Carabiner1 [{
   shapeType=Capsule
   length=3.0127
   radius=1.9873
   shapeName=dyn_Carabiner1_capsule
   center=<1.3059 0.3 0>
   rotation=<rot 0.0001 0.7389 -90.0006>
  }]
 dyn_Carabiner_MSTR [{shapeType=Capsule, length=0.505, radius=0.505, shapeName=dyn_Carabiner_MSTR_capsule, rotation=<rot 0 -90 0>}]
 dyn_Chain1 [{shapeType=Capsule, length=1.4092, radius=1.7187, shapeName=dyn_Chain1_capsule, center=<1.0994 0 0>, rotation=<rot 90 180 176.3559>}]
 dyn_Chain2 [{shapeType=Capsule, length=1.1105, radius=1.6395, shapeName=dyn_Chain2_capsule, center=<1.0825 0 0>, rotation=<rot 90 0 0>}]
 dyn_Chain3 [{shapeType=Capsule, length=1.6015, radius=1.264, shapeName=dyn_Chain3_capsule, center=<1.1599 0 0>, rotation=<rot 90 0 0>}]
 dyn_Chain4 [{shapeType=Capsule, length=1.3779, radius=1.378, shapeName=dyn_Chain4_capsule, center=<1.5 -0.5 0.2>, rotation=<rot 90 180 -173.4782>}]
 dyn_StickA1 [{shapeType=Capsule, length=10, radius=1.5, shapeName=dyn_StickA1_capsule, center=<0.8 0 -8.55>, rotation=<rot 90 0 0>}]
 dyn_Strap1 [{
   shapeType=Capsule
   length=2.3657
   radius=1.6167
   shapeName=dyn_Strap1_capsule
   center=<2.3765 0 -0.5>
   rotation=<rot -81.4767 0.0848 -0.0857>
  }]
 dyn_pack [{
   shapeType=Capsule
   length=12
   radius=7.75
   shapeName=dyn_pack_capsule
   center=<-0.4198 -6.8684 -6.9805>
   rotation=<rot 8.9079 0.9755 -0.3768>
  }]
 pack [{shapeType=Capsule, length=2, radius=2, shapeName=pack_capsule}]
 root [{shapeType=Capsule, length=2, radius=2, shapeName=root_capsule}]
}
constraints [
 {
  bone1Name=pack
  bone2Name=root
  swing1LimitDegrees=45
  swing1Motion=Locked
  swing2LimitDegrees=45
  swing2Motion=Locked
  twistLimitDegrees=45
  twistMotion=Locked
 }
 {
  bone1Name=dyn_pack
  bone2Name=pack
  swing1LimitDegrees=45
  swing1Motion=Locked
  swing2LimitDegrees=45
  swing2Motion=Locked
  twistLimitDegrees=45
  twistMotion=Locked
 }
 {
  bone1Name=dyn_Carabiner_MSTR
  bone2Name=dyn_pack
  swing1LimitDegrees=45
  swing1Motion=Free
  swing2LimitDegrees=45
  swing2Motion=Free
  twistLimitDegrees=45
  twistMotion=Free
 }
 {
  bone1Name=dyn_Carabiner1
  bone2Name=dyn_Carabiner_MSTR
  swing1LimitDegrees=45
  swing1Motion=Locked
  swing2LimitDegrees=50
  twistLimitDegrees=10
 }
 {bone1Name=dyn_Chain1, bone2Name=dyn_Carabiner1, swing1LimitDegrees=50, swing2LimitDegrees=50, twistLimitDegrees=10}
 {bone1Name=dyn_Chain2, bone2Name=dyn_Chain1, swing1LimitDegrees=50, swing2LimitDegrees=50, twistLimitDegrees=10}
 {bone1Name=dyn_Chain3, bone2Name=dyn_Chain2, swing1LimitDegrees=50, swing2LimitDegrees=50, twistLimitDegrees=10}
 {bone1Name=dyn_Chain4, bone2Name=dyn_Chain3, swing1LimitDegrees=50, swing2LimitDegrees=50, twistLimitDegrees=10}
 {bone1Name=dyn_StickA1, bone2Name=dyn_pack, swing1LimitDegrees=10, swing2LimitDegrees=10, twistLimitDegrees=0}
 {bone1Name=dyn_Strap1, bone2Name=dyn_pack, swing1LimitDegrees=0, swing2LimitDegrees=15, twistLimitDegrees=10}
]
