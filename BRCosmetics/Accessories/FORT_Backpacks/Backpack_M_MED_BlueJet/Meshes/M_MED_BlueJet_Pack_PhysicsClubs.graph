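# Physics asset for the M_MED_BlueJet backpack mesh, in a clubs variant, that defines simple capsule bodies for its bone chain.
# Capsules on root, pack, dyn_pack and dyn_StickA1 are chained by constraints, with dyn_pack carrying the large main-body capsule.
# Pack and dyn_pack have all angular motion locked to their parents, while dyn_StickA1 has a small swing limit and zero twist limit against dyn_pack.
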PhysicsAsset /BRCosmetics/Accessories/FORT_Backpacks/Backpack_M_MED_BlueJet/Meshes/M_MED_BlueJet_Pack_PhysicsClubs
bodies {
 dyn_StickA1 [{shapeType=Capsule, length=10, radius=1.5, shapeName=dyn_StickA1_capsule, center=<0.8 0 0>, rotation=<rot 90 0 0>}]
 dyn_pack [{
   shapeType=Capsule
   length=12
   radius=7.75
   shapeName=dyn_pack_capsule
   center=<-0.4198 -6.8684 -6.9805>
   rotation=<rot 8.9079 0.9755 -0.3768>
  }]
 pack [{shapeType=Capsule, length=2, radius=2, shapeName=pack_capsule}]
 root [{shapeType=Capsule, length=2, radius=2, shapeName=root_capsule}]
}
constraints [
 {
  bone1Name=pack
  bone2Name=root
  swing1LimitDegrees=45
  swing1Motion=Locked
  swing2LimitDegrees=45
  swing2Motion=Locked
  twistLimitDegrees=45
  twistMotion=Locked
 }
 {
  bone1Name=dyn_pack
  bone2Name=pack
  swing1LimitDegrees=45
  swing1Motion=Locked
  swing2LimitDegrees=45
  swing2Motion=Locked
  twistLimitDegrees=45
  twistMotion=Locked
 }
 {bone1Name=dyn_StickA1, bone2Name=dyn_pack, swing1LimitDegrees=10, swing2LimitDegrees=10, twistLimitDegrees=0}
]
