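# Animation Blueprint for the BlueJet backpack cosmetic, adding secondary motion and reacting to its matching pickaxe.
# Chains gravity-overridden rigid-body physics, a clubs simulation disabled while that pickaxe is equipped, a pack look-at toward the body's pelvis, and an additive skydiving offset.
# Each update refreshes gravity, the tag-based melee flag and pelvis location.
# Weapon, customization and skydiving events toggle a material parameter hiding part of the pack.
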AnimBlueprint /BRCosmetics/Accessories/FORT_Backpacks/Backpack_M_MED_BlueJet/Meshes/M_MED_BlueJet_Pack_AnimBP
target_skeleton=&/M_MED_BlueJet_Pack_Skeleton
graphs [
 AnimGraph {nodes [
   n1 = RigidBody!(
    ComponentPose=Local To Component(LocalPose=Input Pose())
    bAlphaBoolEnabled:bool
    OverrideWorldGravity:Vector=Get GravityOverride()
    ExternalForce
   ) -> (Pose) {
    alphaInputType=Bool
    bClampLinearTranslationLimitToRefPose=true
    bOverrideWorldGravity=true
    overridePhysicsAsset=&/M_MED_BlueJet_Pack_Physics1
    simulationSpace=BaseBoneSpace
    alphaBoolBlend {blendOption=Linear}
    baseBoneRef {boneName=root}
    componentAppliedLinearAccClamp=<6000 6000 6000>
    componentLinearAccScale=<0.5 0.5 0.5>
    componentLinearVelScale=<0.5 0.5 0.5>
   }
   n2 = RigidBody(ComponentPose=@n1, bAlphaBoolEnabled=NOT Boolean(A=Get IsMelee() -> (IsMelee:bool)), OverrideWorldGravity="0,0,0") {
    alphaInputType=Bool
    bClampLinearTranslationLimitToRefPose=true
    bOverrideWorldGravity=true
    overridePhysicsAsset=&/M_MED_BlueJet_Pack_PhysicsClubs
    simulationSpace=BaseBoneSpace
    alphaBoolBlend {blendOption=Linear}
    baseBoneRef {boneName=root}
    componentAppliedLinearAccClamp=<6000 6000 6000>
    componentLinearAccScale=<0.5 0.5 0.5>
    componentLinearVelScale=<0.5 0.5 0.5>
   }
   n3 = Look At - Bone: pack!(ComponentPose=@n2, bAlphaBoolEnabled:bool=True, LookAtLocation:Vector=Get PelvisLocation() -> (PelvisLocation:Vector)) -> (Pose) {
    alphaInputType=Bool
    interpolationType=EaseInOutExponent2
    alphaBoolBlend {blendOption=Linear}
    boneToModify {boneName=pack}
    lookAt_Axis {axis {y=0, z=-1}}
    lookAtTarget=~
   }
   n4 = "Transform (Modify) Bone - Bone: dyn_pack"!(ComponentPose=@n3, bAlphaBoolEnabled:bool=Get bIsSkyDiving(), Translation="0,-2.5,0", Rotation="10,0,10", Scale="1,1,1") -> (Pose) {
    alphaInputType=Bool
    rotationMode=BMM_Additive
    rotationSpace=BCS_BoneSpace
    translationMode=BMM_Additive
    translationSpace=BCS_BoneSpace
    alphaBoolBlend {blendOption=Linear}
    boneToModify {boneName=dyn_pack}
    rotation {pitch=10, roll=10}
    translation {y=-2.5}
   }
   Output Pose(Result=Component To Local(ComponentPose=@n4)) {layerGroup=None}
  ]}
 EventGraph {nodes [
   Event BlueprintUpdateAnimation()
   > n1 = Sequence!(execute) -> (then_0, then_1, then_2:Exec)
   > n2 = Gravity Override State Function DEPRECATED(CharacterPartAnimInstance=Self-Reference(), GameTimeInSeconds=GetGameTimeInSeconds(), GravityOverrideParams=Get Gravity Override Params())
   > Set GravityOverride(GravityOverride=@n2)
   n3 = PickaxeItemDefinitionHasAnyMetaTag(Tags=Make Array("[0]"=(TagName="Cosmetics.MetaTags.BlueJet")), bRequiresEquipped=true)
   n4 = Branch(execute=@n1.then_1, Condition=@n3)
   > Set IsMelee(IsMelee:bool=true) -> (Output_Get:bool)
   Set IsMelee(execute=@n4.else, IsMelee=false)
   n5 = Get BodyMesh() -> (BodyMesh:Skeletal Mesh Component Object Reference)
   Is Valid(exec=@n1.then_2, InputObject=@n5)
   > Set PelvisLocation(PelvisLocation:Vector=GetSocketLocation!(self:Scene Component Object Reference=@n5, InSocketName:Name=pelvis) -> (ReturnValue:Vector)) -> (Output_Get:Vector)
   n6 = Event OnWeaponChanged()
   n7 = Get Dynamic Material Instance(execute=[@n6.then, @n15.then], SkeletalMeshComponent=GetPartSkeletalMeshComponent())
   > n8 = Sequence()
   > n9 = Branch(Condition=@n3)
   > n10 = Branch(Condition=Get bIsSkyDiving())
   SetScalarParameterValue(execute=[@n13.then, @n10.then, @n11.else, @n9.else], self=@n7, ParameterName=Hide Element 01, Value=0)
   SetScalarParameterValue(execute=[@n10.else, @n11.then], self=@n7, ParameterName=Hide Element 01, Value=1)
   n11 = Branch(execute=[@n14.then, @n8.then_1], Condition=@n3)
   n12 = Cast To FortPlayerPawnAthena(execute=@n8.then_2, Object=TryGetPawnOwner())
   > Set BodyMesh(BodyMesh=GetSkeletalMeshForPartType(self=@n12))
   n13 = Event OnBeginSkydiving()
   n14 = Event OnEndSkydiving()
   n15 = Event OnCharacterCustomizationLoaded()
   Switch on EFortWeaponCoreAnimation!(
    execute
    Selection:EFortWeaponCoreAnimation Enum=GetWeaponCoreAnim!(self:Fort Player Animation Interface Object Reference=GetOwnerAnimationInterface!(self:Custom Character Part Anim Instance Object Reference) -> (ReturnValue:Fort Player Animation Interface Object Reference)) -> (ReturnValue:EFortWeaponCoreAnimation Enum)
   ) -> (
    Melee:Exec
    Pistol:Exec
    Shotgun:Exec
    PaperBlueprint:Exec
    Rifle:Exec
    MeleeOneHand:Exec
    MachinePistol:Exec
    RocketLauncher:Exec
    GrenadeLauncher:Exec
    GoingCommando:Exec
    AssaultRifle:Exec
    TacticalShotgun:Exec
    SniperRifle:Exec
    TrapPlacement:Exec
    ShoulderLauncher:Exec
    AbilityDecoTool:Exec
    Crossbow:Exec
    C4:Exec
    RemoteControl:Exec
    DualWield:Exec
    AR_BullPup:Exec
    AR_ForwardGrip:Exec
    MedPackPaddles:Exec
    SMG_P90:Exec
    AR_DrumGun:Exec
    Consumable_Small:Exec
    Consumable_Large:Exec
    Balloon:Exec
    MountedTurret:Exec
    CreativeTool:Exec
    ExplosiveBow:Exec
    AshtonIndigo:Exec
    AshtonChicago:Exec
    MeleeDualWield:Exec
    Unarmed:Exec
   )
   > Set IsMelee(IsMelee=true)
   TryGetPawnOwner()
  ]}
 NewFunction {nodes [NewFunction() {is=title K2Node_FunctionEntry}]}
]
variables [
 BodyMesh {type=SkeletalMeshComponent}
 DynamicMaterialInstance {type=MaterialInstance, default=null}
 Gravity Override Params {type=GravityOverrideParamsStruct}
 GravityOverride {type=Vector}
 IsMelee {type=boolean}
 PelvisLocation {type=Vector}
 SkeletalMesh {type=SkeletalMeshComponent}
]
